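# Cheap physics asset for the bags on the M_MED_Dirty_Docks male player body, covering a single right-side pouch.
# A box body on dyn_pouch_01_r is constrained to a tiny capsule on its MSTR bone, which serves as an anchor.
# The joint locks swing 1 and sets small angle limits for swing 2 and twist.
# The right thigh twist bone carries a sphere and capsule approximating the leg beside the pouch.
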PhysicsAsset /BRCosmetics/Characters/Player/Male/Medium/Bodies/M_MED_Dirty_Docks/Meshes/M_MED_Dirty_Docks_Bags_Cheap_Physics
bodies {
 dyn_pouch_01_r [{shapeType=Box, extentX=7.01, extentY=14.01, extentZ=14.01, center=<-17.5647 -0.5647 -1.5903>, rotation=<rot 39.9953 0 -180>}]
 dyn_pouch_01_r_MSTR [{shapeType=Capsule, length=0.01, radius=0.01, rotation=<rot 0 -90 0>}]
 thigh_twist_01_r [
  {shapeType=Sphere, radius=11, center=<-22.9643 3.5896 -1.6899>}
  {shapeType=Capsule, length=29.9357, radius=8.1289, center=<-13.1519 -0.0021 -0.6572>, rotation=<rot 87.1746 180 0>}
 ]
}
constraints [{
  bone1Name=dyn_pouch_01_r
  bone2Name=dyn_pouch_01_r_MSTR
  swing1LimitDegrees=10
  swing1Motion=Locked
  swing2LimitDegrees=12
  twistLimitDegrees=8
 }]
