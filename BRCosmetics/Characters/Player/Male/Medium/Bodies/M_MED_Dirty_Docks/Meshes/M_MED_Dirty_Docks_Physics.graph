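# Physics asset for the Dirty Docks male medium player body cosmetic, covering secondary motion of two dangling strings and left and right zipper pulls.
# Small parent capsules head a three-link string A, a four-link string B and each single-link zipper, with narrow twist and moderate swing limits.
# A large spine_05 capsule stands in for the upper torso.
# Only the string parents are constrained to it, with free rotation; zipper parents have no such constraint.
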PhysicsAsset /BRCosmetics/Characters/Player/Male/Medium/Bodies/M_MED_Dirty_Docks/Meshes/M_MED_Dirty_Docks_Physics
bodies {
 dyn_parent_stringA_01 [{shapeType=Capsule, length=0.505, radius=0.505, rotation=<rot 0 -90 0>}]
 dyn_parent_stringB_01 [{shapeType=Capsule, length=0.505, radius=0.505, rotation=<rot 0 -90 0>}]
 dyn_parent_zipper_l [{shapeType=Capsule, length=0.505, radius=0.505, rotation=<rot 0 -90 0>}]
 dyn_parent_zipper_r [{shapeType=Capsule, length=0.505, radius=0.505, rotation=<rot 0 -90 0>}]
 dyn_stringA_01 [{shapeType=Capsule, length=2.3, radius=0.75, center=<1.9727 0 0.8319>, rotation=<rot 87.0854 179.9999 180>}]
 dyn_stringA_02 [{shapeType=Capsule, length=3.3, radius=0.5, center=<2 0 0.2296>, rotation=<rot 90 0 0>}]
 dyn_stringA_03 [{shapeType=Capsule, length=1.8, radius=0.5, center=<1.9921 0.0959 0.2085>, rotation=<rot 90 0 -4.6842>}]
 dyn_stringB_01 [{shapeType=Capsule, length=3.25, radius=0.8, center=<1.736 -0.263 0>, rotation=<rot 90 0 0>}]
 dyn_stringB_02 [{shapeType=Capsule, length=2.55, radius=0.75, center=<2 -0.299 0>, rotation=<rot 90 0 0.6091>}]
 dyn_stringB_03 [{shapeType=Capsule, length=1.8, radius=0.55, center=<2 -0.1697 0>, rotation=<rot 90 0 0>}]
 dyn_stringB_04 [{shapeType=Capsule, length=1.25, radius=0.55, center=<1.1708 -0.1977 0>, rotation=<rot 90 0 0>}]
 dyn_zipper_l [{shapeType=Capsule, length=2.0983, radius=0.8144, center=<1.8858 -0.0018 0.0173>, rotation=<rot 0 91.2681 -88.6093>}]
 dyn_zipper_r [{shapeType=Capsule, length=2.0983, radius=0.8144, center=<1.8845 -0.0473 -0.0572>, rotation=<rot 0 -90.1113 -87.3986>}]
 spine_05 [{shapeType=Capsule, length=1, radius=22.5, center=<-7.1228 -10.0948 0>, rotation=<rot 0 15.9042 0>}]
}
constraints [
 {
  bone1Name=dyn_parent_stringA_01
  bone2Name=spine_05
  swing1LimitDegrees=45
  swing1Motion=Free
  swing2LimitDegrees=45
  swing2Motion=Free
  twistLimitDegrees=45
  twistMotion=Free
 }
 {
  bone1Name=dyn_parent_stringB_01
  bone2Name=spine_05
  swing1LimitDegrees=45
  swing1Motion=Free
  swing2LimitDegrees=45
  swing2Motion=Free
  twistLimitDegrees=45
  twistMotion=Free
 }
 {bone1Name=dyn_stringB_01, bone2Name=dyn_parent_stringB_01, swing1LimitDegrees=20, swing2LimitDegrees=35, twistLimitDegrees=5}
 {bone1Name=dyn_stringB_02, bone2Name=dyn_stringB_01, swing1LimitDegrees=20, swing2LimitDegrees=35, twistLimitDegrees=5}
 {bone1Name=dyn_stringB_03, bone2Name=dyn_stringB_02, swing1LimitDegrees=20, swing2LimitDegrees=40, twistLimitDegrees=5}
 {bone1Name=dyn_stringB_04, bone2Name=dyn_stringB_03, swing1LimitDegrees=20, swing2LimitDegrees=45, twistLimitDegrees=5}
 {bone1Name=dyn_stringA_03, bone2Name=dyn_stringA_02, swing1LimitDegrees=15, swing2LimitDegrees=35, twistLimitDegrees=5}
 {bone1Name=dyn_stringA_02, bone2Name=dyn_stringA_01, swing1LimitDegrees=15, swing2LimitDegrees=35, twistLimitDegrees=5}
 {bone1Name=dyn_stringA_01, bone2Name=dyn_parent_stringA_01, swing1LimitDegrees=15, swing2LimitDegrees=35, twistLimitDegrees=5}
 {bone1Name=dyn_zipper_l, bone2Name=dyn_parent_zipper_l, swing1LimitDegrees=25, swing2LimitDegrees=25, twistLimitDegrees=5}
 {bone1Name=dyn_zipper_r, bone2Name=dyn_parent_zipper_r, swing1LimitDegrees=25, swing2LimitDegrees=25, twistLimitDegrees=5}
]
