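# Physics asset for the shirt mesh of the medium male Dirty Docks player cosmetic body, defining bodies for secondary motion of the shirt.
# Ten dyn_shirt capsules are spread over front, side and back on both left and right.
# Each is constrained to its own small dyn_parent capsule with identical narrow swing and twist limits.
# No constraints link neighboring segments, so each parent-child pair is jointed independently.
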PhysicsAsset /BRCosmetics/Characters/Player/Male/Medium/Bodies/M_MED_Dirty_Docks/Meshes/M_MED_Dirty_Docks_Shirt_Physics
bodies {
 dyn_parent_shirt_b_l [{shapeType=Capsule, length=0.505, radius=0.505, rotation=<rot 0 -90 0>}]
 dyn_parent_shirt_b_r [{shapeType=Capsule, length=0.505, radius=0.505, rotation=<rot 0 -90 0>}]
 dyn_parent_shirt_bck_l [{shapeType=Capsule, length=0.505, radius=0.505, rotation=<rot 0 -90 0>}]
 dyn_parent_shirt_bck_r [{shapeType=Capsule, length=0.505, radius=0.505, rotation=<rot 0 -90 0>}]
 dyn_parent_shirt_f_l [{shapeType=Capsule, length=0.505, radius=0.505, rotation=<rot 0 -90 0>}]
 dyn_parent_shirt_f_r [{shapeType=Capsule, length=0.505, radius=0.505, rotation=<rot 0 -90 0>}]
 dyn_parent_shirt_frt_l [{shapeType=Capsule, length=0.505, radius=0.505, rotation=<rot 0 -90 0>}]
 dyn_parent_shirt_frt_r [{shapeType=Capsule, length=0.505, radius=0.505, rotation=<rot 0 -90 0>}]
 dyn_parent_shirt_l [{shapeType=Capsule, length=0.505, radius=0.505, rotation=<rot 0 -90 0>}]
 dyn_parent_shirt_r [{shapeType=Capsule, length=0.505, radius=0.505, rotation=<rot 0 -90 0>}]
 dyn_shirt_b_l [{shapeType=Capsule, length=3, radius=1.75, center=<2 0 0>, rotation=<rot 90 0 0>}]
 dyn_shirt_b_r [{shapeType=Capsule, length=3, radius=1.75, center=<2 0 0>, rotation=<rot 90 0 0>}]
 dyn_shirt_bck_l [{shapeType=Capsule, length=3, radius=1.75, center=<2 0 0>, rotation=<rot 90 0 0>}]
 dyn_shirt_bck_r [{shapeType=Capsule, length=3, radius=1.75, center=<2 0 0>, rotation=<rot 90 0 0>}]
 dyn_shirt_f_l [{shapeType=Capsule, length=3, radius=1.75, center=<2 0 0>, rotation=<rot 90 0 0>}]
 dyn_shirt_f_r [{shapeType=Capsule, length=3, radius=1.75, center=<2 0 0>, rotation=<rot 90 0 0>}]
 dyn_shirt_frt_l [{shapeType=Capsule, length=3, radius=1.75, center=<2 0 0>, rotation=<rot 90 0 0>}]
 dyn_shirt_frt_r [{shapeType=Capsule, length=3, radius=1.75, center=<2 0 0>, rotation=<rot 90 0 0>}]
 dyn_shirt_l [{shapeType=Capsule, length=3, radius=1.75, center=<2 0 0>, rotation=<rot 90 0 0>}]
 dyn_shirt_r [{shapeType=Capsule, length=3, radius=1.75, center=<2 0 0>, rotation=<rot 90 0 0>}]
}
constraints [
 {bone1Name=dyn_shirt_frt_l, bone2Name=dyn_parent_shirt_frt_l, swing1LimitDegrees=8, swing2LimitDegrees=13, twistLimitDegrees=5}
 {bone1Name=dyn_shirt_f_l, bone2Name=dyn_parent_shirt_f_l, swing1LimitDegrees=8, swing2LimitDegrees=13, twistLimitDegrees=5}
 {bone1Name=dyn_shirt_l, bone2Name=dyn_parent_shirt_l, swing1LimitDegrees=8, swing2LimitDegrees=13, twistLimitDegrees=5}
 {bone1Name=dyn_shirt_b_l, bone2Name=dyn_parent_shirt_b_l, swing1LimitDegrees=8, swing2LimitDegrees=13, twistLimitDegrees=5}
 {bone1Name=dyn_shirt_bck_l, bone2Name=dyn_parent_shirt_bck_l, swing1LimitDegrees=8, swing2LimitDegrees=13, twistLimitDegrees=5}
 {bone1Name=dyn_shirt_frt_r, bone2Name=dyn_parent_shirt_frt_r, swing1LimitDegrees=8, swing2LimitDegrees=13, twistLimitDegrees=5}
 {bone1Name=dyn_shirt_f_r, bone2Name=dyn_parent_shirt_f_r, swing1LimitDegrees=8, swing2LimitDegrees=13, twistLimitDegrees=5}
 {bone1Name=dyn_shirt_r, bone2Name=dyn_parent_shirt_r, swing1LimitDegrees=8, swing2LimitDegrees=13, twistLimitDegrees=5}
 {bone1Name=dyn_shirt_b_r, bone2Name=dyn_parent_shirt_b_r, swing1LimitDegrees=8, swing2LimitDegrees=13, twistLimitDegrees=5}
 {bone1Name=dyn_shirt_bck_r, bone2Name=dyn_parent_shirt_bck_r, swing1LimitDegrees=8, swing2LimitDegrees=13, twistLimitDegrees=5}
]
