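# Cosmetic AnimBP for the M_MED_Dirty_Docks player body, adding secondary motion to shirt, zipper, pouches and scarf collar.
# It copies the parent mesh pose, then runs weighted bone constraints, additive scarf and pouch offsets, and three RigidBody simulations.
# The EventGraph derives offset alphas from head and thigh rotation, pouch simulation alpha and gravity from movement, skydiving and emotes.
# Crouch and skydive transitions reset dynamics.
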AnimBlueprint /BRCosmetics/Characters/Player/Male/Medium/Bodies/M_MED_Dirty_Docks/Meshes/M_MED_Dirty_Docks_AnimBP
target_skeleton=&/M_MED_Dirty_Docks_Skeleton
graphs [
 AnimGraph {nodes [
   n1 = "Constraint - dyn_parent_shirt_frt_l (PPPP)"!(
    ComponentPose=Local To Component(LocalPose=Copy Pose From Mesh() {bUseAttachedParent=true})
    Alpha=1
    ConstraintWeights_0=0.0433
    ConstraintWeights_1:float=0.1712
    ConstraintWeights_2:float=0.5428
    ConstraintWeights_3:float=0.2427
   ) -> (Pose) {
    alphaBoolBlend {blendOption=Linear}
    boneToModify {boneName=dyn_parent_shirt_frt_l}
    constraintSetup [{targetBone {boneName=spine_02}}, {targetBone {boneName=spine_03}}, {targetBone {boneName=spine_04}}, {targetBone {boneName=spine_05}}]
    constraintWeights [1,1,1,1]
   }
   n2 = "Constraint - dyn_parent_shirt_f_l (PPPPP)"!(
    ComponentPose=@n1
    Alpha=1
    ConstraintWeights_0=0.0199
    ConstraintWeights_1:float=0.1962
    ConstraintWeights_2:float=0.6227
    ConstraintWeights_3:float=0.0764
    ConstraintWeights_4:float=0.0848
   ) -> (Pose) {
    alphaBoolBlend {blendOption=Linear}
    boneToModify {boneName=dyn_parent_shirt_f_l}
    constraintSetup [
     {targetBone {boneName=spine_02}}
     {targetBone {boneName=spine_03}}
     {targetBone {boneName=spine_04}}
     {targetBone {boneName=spine_05}}
     {targetBone {boneName=dyn_shirt_f_l}}
    ]
    constraintWeights [1,1,1,1,1]
   }
   n3 = "Constraint - dyn_parent_shirt_l (PPPPPP)"!(
    ComponentPose=@n2
    Alpha=1
    ConstraintWeights_0=0.0439
    ConstraintWeights_1:float=0.0118
    ConstraintWeights_2:float=0.3243
    ConstraintWeights_3:float=0.5378
    ConstraintWeights_4:float=0.0449
    ConstraintWeights_5:float=0.0372
   ) -> (Pose) {
    alphaBoolBlend {blendOption=Linear}
    boneToModify {boneName=dyn_parent_shirt_l}
    constraintSetup [
     {targetBone {boneName=clavicle_l}}
     {targetBone {boneName=spine_02}}
     {targetBone {boneName=spine_03}}
     {targetBone {boneName=spine_04}}
     {targetBone {boneName=spine_05}}
     {targetBone {boneName=upperarm_twist_01_l}}
    ]
    constraintWeights [1,1,1,1,1,1]
   }
   n4 = "Constraint - dyn_parent_shirt_b_l (PPPPP)"!(
    ComponentPose=@n3
    Alpha=1
    ConstraintWeights_0=0.0221
    ConstraintWeights_1:float=0.36
    ConstraintWeights_2:float=0.4959
    ConstraintWeights_3:float=0.1098
    ConstraintWeights_4:float=0.0123
   ) -> (Pose) {
    alphaBoolBlend {blendOption=Linear}
    boneToModify {boneName=dyn_parent_shirt_b_l}
    constraintSetup [
     {targetBone {boneName=spine_02}}
     {targetBone {boneName=spine_03}}
     {targetBone {boneName=spine_04}}
     {targetBone {boneName=spine_05}}
     {targetBone {boneName=clavicle_l}}
    ]
    constraintWeights [1,1,1,1,1]
   }
   n5 = "Constraint - dyn_parent_shirt_bck_l (PPP)"!(ComponentPose=@n4, Alpha=1, ConstraintWeights_0=0.2476, ConstraintWeights_1:float=0.3907, ConstraintWeights_2:float=0.3616) -> (Pose) {
    alphaBoolBlend {blendOption=Linear}
    boneToModify {boneName=dyn_parent_shirt_bck_l}
    constraintSetup [{targetBone {boneName=spine_03}}, {targetBone {boneName=spine_04}}, {targetBone {boneName=spine_05}}]
    constraintWeights [1,1,1]
   }
   n6 = "Constraint - dyn_parent_shirt_frt_r (PPPPP)"!(
    ComponentPose=@n5
    Alpha=1
    ConstraintWeights_0=0.1482
    ConstraintWeights_1:float=0.0364
    ConstraintWeights_2:float=0.1448
    ConstraintWeights_3:float=0.4559
    ConstraintWeights_4:float=0.2147
   ) -> (Pose) {
    alphaBoolBlend {blendOption=Linear}
    boneToModify {boneName=dyn_parent_shirt_frt_r}
    constraintSetup [
     {targetBone {boneName=dyn_shirt_frt_r}}
     {targetBone {boneName=spine_02}}
     {targetBone {boneName=spine_03}}
     {targetBone {boneName=spine_04}}
     {targetBone {boneName=spine_05}}
    ]
    constraintWeights [1,1,1,1,1]
   }
   n7 = "Constraint - dyn_parent_shirt_f_r (PPPPPPPP)"!(
    ComponentPose=@n6
    Alpha=1
    ConstraintWeights_0=0.3215
    ConstraintWeights_1:float=0.0139
    ConstraintWeights_2:float=0.0176
    ConstraintWeights_3:float=0.1043
    ConstraintWeights_4:float=0.1093
    ConstraintWeights_5:float=0.3528
    ConstraintWeights_6:float=0.0589
    ConstraintWeights_7:float=0.0217
   ) -> (Pose) {
    alphaBoolBlend {blendOption=Linear}
    boneToModify {boneName=dyn_parent_shirt_f_r}
    constraintSetup [
     {targetBone {boneName=dyn_shirt_f_r}}
     {targetBone {boneName=upperarm_twist_01_r}}
     {targetBone {boneName=dyn_shirt_frt_r}}
     {targetBone {boneName=dyn_shirt_r}}
     {targetBone {boneName=spine_03}}
     {targetBone {boneName=spine_04}}
     {targetBone {boneName=spine_05}}
     {targetBone {boneName=clavicle_r}}
    ]
    constraintWeights [1,1,1,1,1,1,1,1]
   }
   n8 = "Constraint - dyn_parent_shirt_r (PPPPPPPP)"!(
    ComponentPose=@n7
    Alpha=1
    ConstraintWeights_0=0.0246
    ConstraintWeights_1:float=0.0333
    ConstraintWeights_2:float=0.1142
    ConstraintWeights_3:float=0.0104
    ConstraintWeights_4:float=0.2815
    ConstraintWeights_5:float=0.4601
    ConstraintWeights_6:float=0.037
    ConstraintWeights_7:float=0.0388
   ) -> (Pose) {
    alphaBoolBlend {blendOption=Linear}
    boneToModify {boneName=dyn_parent_shirt_r}
    constraintSetup [
     {targetBone {boneName=dyn_shirt_f_r}}
     {targetBone {boneName=upperarm_twist_01_r}}
     {targetBone {boneName=dyn_shirt_r}}
     {targetBone {boneName=spine_02}}
     {targetBone {boneName=spine_03}}
     {targetBone {boneName=spine_04}}
     {targetBone {boneName=spine_05}}
     {targetBone {boneName=clavicle_r}}
    ]
    constraintWeights [1,1,1,1,1,1,1,1]
   }
   n9 = "Constraint - dyn_parent_shirt_b_r (PPPPP)"!(
    ComponentPose=@n8
    Alpha=1
    ConstraintWeights_0=0.02
    ConstraintWeights_1:float=0.0221
    ConstraintWeights_2:float=0.355
    ConstraintWeights_3:float=0.4904
    ConstraintWeights_4:float=0.1125
   ) -> (Pose) {
    alphaBoolBlend {blendOption=Linear}
    boneToModify {boneName=dyn_parent_shirt_b_r}
    constraintSetup [
     {targetBone {boneName=clavicle_r}}
     {targetBone {boneName=spine_02}}
     {targetBone {boneName=spine_03}}
     {targetBone {boneName=spine_04}}
     {targetBone {boneName=spine_05}}
    ]
    constraintWeights [1,1,1,1,1]
   }
   n10 = "Constraint - dyn_parent_shirt_bck_r (PPP)"!(ComponentPose=@n9, Alpha=1, ConstraintWeights_0=0.2486, ConstraintWeights_1:float=0.3904, ConstraintWeights_2:float=0.361) -> (Pose) {
    alphaBoolBlend {blendOption=Linear}
    boneToModify {boneName=dyn_parent_shirt_bck_r}
    constraintSetup [{targetBone {boneName=spine_03}}, {targetBone {boneName=spine_04}}, {targetBone {boneName=spine_05}}]
    constraintWeights [1,1,1]
   }
   n11 = "Constraint - dyn_parent_zipper_l (PPP)"!(ComponentPose=@n10, Alpha=1, ConstraintWeights_0=0.6906, ConstraintWeights_1:float=0.0229, ConstraintWeights_2:float=0.2864) -> (Pose) {
    alphaBoolBlend {blendOption=Linear}
    boneToModify {boneName=dyn_parent_zipper_l}
    constraintSetup [{targetBone {boneName=pelvis}}, {targetBone {boneName=spine_02}}, {targetBone {boneName=spine_01}}]
    constraintWeights [1,1,1]
   }
   n12 = "Constraint - dyn_parent_zipper_r (PPPP)"!(
    ComponentPose=@n11
    Alpha=1
    ConstraintWeights_0=0.6845
    ConstraintWeights_1:float=0.0238
    ConstraintWeights_2:float=0.0105
    ConstraintWeights_3:float=0.2812
   ) -> (Pose) {
    alphaBoolBlend {blendOption=Linear}
    boneToModify {boneName=dyn_parent_zipper_r}
    constraintSetup [{targetBone {boneName=pelvis}}, {targetBone {boneName=spine_02}}, {targetBone {boneName=spine_03}}, {targetBone {boneName=spine_01}}]
    constraintWeights [1,1,1,1]
   }
   n13 = "Transform (Modify) Bone - Bone: dyn_pouchB"(ComponentPose=@n12, Alpha=Get thigh_rot_r(), Translation="4.5,0.000,-0.000", Rotation="0,0,0", Scale="1,1,1") {
    translationMode=BMM_Additive
    translationSpace=BCS_BoneSpace
    alphaBoolBlend {blendOption=Linear}
    alphaScaleBiasClamp {bInterpResult=true, interpSpeedDecreasing=25, interpSpeedIncreasing=25}
    boneToModify {boneName=dyn_pouchB}
    translation {x=4.5}
   }
   n14 = "Constraint - sim (PP)"!(
    ComponentPose=@n13
    Alpha=1
    ConstraintWeights_0=Get SkydiveRootSpace() -> (SkydiveRootSpace:double)
    ConstraintWeights_1:float=Get SkydiveSpineSpace() -> (SkydiveSpineSpace:double)
   ) -> (Pose) {
    alphaBoolBlend {blendOption=Linear}
    boneToModify {boneName=sim}
    constraintSetup [{targetBone {boneName=root}}, {targetBone {boneName=spine_01}}]
    constraintWeights [1,1]
   }
   n15 = "Constraint - dyn_pouch_01_r_MSTR (PP)"!(ComponentPose=@n14, Alpha=1, ConstraintWeights_0=0, ConstraintWeights_1:float=1) -> (Pose) {
    alphaBoolBlend {blendOption=Linear}
    boneToModify {boneName=dyn_pouch_01_r_MSTR}
    constraintSetup [{targetBone {boneName=thigh_r}}, {targetBone {boneName=pelvis}}]
    constraintWeights [1,1]
   }
   n16 = "Constraint - dyn_pouch_01_r_MSTR (RR)"!(ComponentPose=@n15, Alpha=1, ConstraintWeights_0=0.5, ConstraintWeights_1:float=0.5) -> (Pose) {
    alphaBoolBlend {blendOption=Linear}
    boneToModify {boneName=dyn_pouch_01_r_MSTR}
    constraintSetup [{transformType=Rotation, targetBone {boneName=thigh_r}}, {transformType=Rotation, targetBone {boneName=pelvis}}]
    constraintWeights [1,1]
   }
   n17 = Look At - Bone: dyn_pouch_bk_01_r(ComponentPose=@n16, Alpha=1) {
    alphaBoolBlend {blendOption=Linear}
    boneToModify {boneName=dyn_pouch_bk_01_r}
    lookAtLocation=<0 1>
    lookAtTarget {boneReference {boneName=dyn_pouch_bk_01_r_aim}}
    lookAt_Axis {axis=<-1 0>}
    lookUp_Axis {axis {x=0, z=1}}
   }
   n18 = "Constraint - dyn_pouch_fr_01_r_MSTR (PP)"!(ComponentPose=@n17, Alpha=1, ConstraintWeights_0=0, ConstraintWeights_1:float=1) -> (Pose) {
    alphaBoolBlend {blendOption=Linear}
    boneToModify {boneName=dyn_pouch_fr_01_r_MSTR}
    constraintSetup [{targetBone {boneName=thigh_l}}, {targetBone {boneName=pelvis}}]
    constraintWeights [1,1]
   }
   n19 = "Transform (Modify) Bone - Bone: dyn_driver_bck_scarf"(ComponentPose=@n18, Alpha=1, Translation="0,0,0", Rotation="-4,0.000,0", Scale="1,1,1") {
    rotationMode=BMM_Additive
    rotationSpace=BCS_BoneSpace
    scaleMode=BMM_Additive
    scaleSpace=BCS_BoneSpace
    translationMode=BMM_Additive
    translationSpace=BCS_BoneSpace
    alphaBoolBlend {blendOption=Linear}
    alphaScaleBiasClamp {interpSpeedDecreasing=18, interpSpeedIncreasing=25}
    boneToModify {boneName=dyn_driver_bck_scarf}
    rotation {pitch=-4}
   }
   n20 = "Transform (Modify) Bone - Bone: dyn_driver_bck_scarf"(ComponentPose=@n19, Alpha=Get collar_back_alpha() -> (collar_back_alpha:double), Translation="1,10,0", Rotation="0,0,0", Scale="1,1,1") {
    rotationMode=BMM_Additive
    rotationSpace=BCS_BoneSpace
    scaleMode=BMM_Additive
    scaleSpace=BCS_BoneSpace
    translationMode=BMM_Additive
    translationSpace=BCS_BoneSpace
    alphaBoolBlend {blendOption=Linear}
    alphaScaleBiasClamp {bInterpResult=true, interpSpeedDecreasing=18, interpSpeedIncreasing=25}
    boneToModify {boneName=dyn_driver_bck_scarf}
    translation=<1 10>
   }
   n21 = "Transform (Modify) Bone - Bone: dyn_driver_frnt_scarf"(
    ComponentPose=@n20
    Alpha=Get collar_front_alpha() -> (collar_front_alpha:double)
    Translation="1,-0.000,0"
    Rotation="19,0.000,0.000"
    Scale="0.9,0.9,0.9"
   ) {
    rotationMode=BMM_Additive
    rotationSpace=BCS_BoneSpace
    scaleMode=BMM_Additive
    scaleSpace=BCS_BoneSpace
    translationMode=BMM_Additive
    translationSpace=BCS_BoneSpace
    alphaBoolBlend {blendOption=Linear}
    alphaScaleBiasClamp {bInterpResult=true, interpSpeedDecreasing=18, interpSpeedIncreasing=25}
    boneToModify {boneName=dyn_driver_frnt_scarf}
    rotation {pitch=19}
    scale=<0.9 0.9 0.9>
    translation {x=1}
   }
   n22 = "Transform (Modify) Bone - Bone: dyn_driver_bck_scarf"(
    ComponentPose=@n21
    Alpha=Get collar_back_left_alpha() -> (collar_back_left_alpha:double)
    Translation="0,0,0"
    Rotation="-5,-12,-10"
    Scale="1,1,1"
   ) {
    rotationMode=BMM_Additive
    rotationSpace=BCS_BoneSpace
    scaleMode=BMM_Additive
    scaleSpace=BCS_BoneSpace
    translationMode=BMM_Additive
    translationSpace=BCS_BoneSpace
    alphaBoolBlend {blendOption=Linear}
    alphaScaleBiasClamp {bInterpResult=true, interpSpeedDecreasing=18, interpSpeedIncreasing=25}
    boneToModify {boneName=dyn_driver_bck_scarf}
    rotation=<rot -5 -12 -10>
   }
   n23 = "Transform (Modify) Bone - Bone: dyn_driver_bck_scarf"(
    ComponentPose=@n22
    Alpha=Get collar_back_right_alpha() -> (collar_back_right_alpha:double)
    Translation="0,0,0"
    Rotation="-5,30,10"
    Scale="1,1,1"
   ) {
    rotationMode=BMM_Additive
    rotationSpace=BCS_BoneSpace
    scaleMode=BMM_Additive
    scaleSpace=BCS_BoneSpace
    translationMode=BMM_Additive
    translationSpace=BCS_BoneSpace
    alphaBoolBlend {blendOption=Linear}
    alphaScaleBiasClamp {bInterpResult=true, interpSpeedDecreasing=18, interpSpeedIncreasing=25}
    boneToModify {boneName=dyn_driver_bck_scarf}
    rotation=<rot -5 30 10>
   }
   n24 = "Transform (Modify) Bone - Bone: dyn_driver_bck_scarf"(
    ComponentPose=@n23
    Alpha=Get collar_back_down_alpha() -> (collar_back_down_alpha:double)
    Translation=Get scarf_back_trans() -> (scarf_back_trans:Vector)
    Rotation=Get scarf_back_rot() -> (scarf_back_rot:Rotator)
    Scale="1,1,1"
   ) {
    rotationMode=BMM_Additive
    rotationSpace=BCS_BoneSpace
    scaleMode=BMM_Additive
    scaleSpace=BCS_BoneSpace
    translationMode=BMM_Additive
    translationSpace=BCS_BoneSpace
    alphaBoolBlend {blendOption=Linear}
    alphaScaleBiasClamp {bInterpResult=true, interpSpeedDecreasing=18, interpSpeedIncreasing=25}
    boneToModify {boneName=dyn_driver_bck_scarf}
   }
   n25 = "Transform (Modify) Bone - Bone: dyn_driver_bck_scarf"(
    ComponentPose=@n24
    Alpha=Get collar_crouch_down_alpha() -> (collar_crouch_down_alpha:double)
    Translation="5,0,2"
    Rotation="-15,0,0"
    Scale="1.02,1.02,1.02"
   ) {
    rotationMode=BMM_Additive
    rotationSpace=BCS_BoneSpace
    scaleMode=BMM_Additive
    scaleSpace=BCS_BoneSpace
    translationMode=BMM_Additive
    translationSpace=BCS_BoneSpace
    alphaBoolBlend {blendOption=Linear}
    alphaScaleBiasClamp {bInterpResult=true, interpSpeedDecreasing=18, interpSpeedIncreasing=25}
    boneToModify {boneName=dyn_driver_bck_scarf}
    rotation {pitch=-15}
    scale=<1.02 1.02 1.02>
    translation {x=5, z=2}
   }
   n26 = RigidBody!(
    ComponentPose=@n25
    Alpha=Get pouch_rigids_alpha() -> (pouch_rigids_alpha:double)
    OverrideWorldGravity:Vector=Get finalGravity() -> (finalGravity:Vector)
    ExternalForce
   ) -> (Pose) {
    bClampLinearTranslationLimitToRefPose=true
    bOverrideWorldGravity=true
    overridePhysicsAsset=&/M_MED_Dirty_Docks_Bags_Cheap_Physics
    simulationSpace=BaseBoneSpace
    alphaBoolBlend {blendOption=Linear}
    alphaScaleBiasClamp {bInterpResult=true, interpSpeedDecreasing=15, interpSpeedIncreasing=15}
    baseBoneRef {boneName=root}
    componentAppliedLinearAccClamp=<1100 1100 1100>
    componentLinearAccScale=<0.5 0.5 0.2>
    componentLinearVelScale=<0.7 0.7 0.2>
    overrideWorldGravity {z=-980}
   }
   n27 = RigidBody!(ComponentPose=@n26, Alpha, ExternalForce) -> (Pose) {
    bClampLinearTranslationLimitToRefPose=true
    bOverrideWorldGravity=true
    overridePhysicsAsset=&/M_MED_Dirty_Docks_Physics
    simulationSpace=BaseBoneSpace
    alphaBoolBlend {blendOption=Linear}
    alphaScaleBiasClamp {bInterpResult=true, interpSpeedDecreasing=15, interpSpeedIncreasing=15}
    baseBoneRef {boneName=root}
    componentAppliedLinearAccClamp=<1100 1100 1100>
    componentLinearAccScale=<0.5 0.5 0.35>
    componentLinearVelScale=<0.7 0.7 0.35>
    overrideWorldGravity {z=-980}
   }
   n28 = RigidBody(ComponentPose=@n27) {
    bClampLinearTranslationLimitToRefPose=true
    bOverrideWorldGravity=true
    overridePhysicsAsset=&/M_MED_Dirty_Docks_Shirt_Physics
    simulationSpace=BaseBoneSpace
    alphaBoolBlend {blendOption=Linear}
    alphaScaleBiasClamp {bInterpResult=true, interpSpeedDecreasing=15, interpSpeedIncreasing=15}
    baseBoneRef {boneName=root}
    componentAppliedLinearAccClamp=<1100 1100 1100>
    componentLinearAccScale=<0.5 0.5 0.35>
    componentLinearVelScale=<0.7 0.7 0.35>
    overrideWorldGravity {z=-980}
   }
   Output Pose(Result=Component To Local(ComponentPose=@n28)) {layerGroup=None}
  ]}
 EventGraph {nodes [
   Event OnBeginJumping()
   > Set jump_bool(jump_bool:bool=true) -> (Output_Get:bool)
   Event OnCharacterCustomizationLoaded()
   n1 = CharacterItemDefinitionHasAnyMetaTag(Tags=Make Array("[0]"=(TagName="Cosmetics.MetaTags.Hat.00")))
   > n2 = Branch(Condition=@n1)
   Set scarf_back_trans(execute=[@n2.then, @n31.then], scarf_back_trans:Vector="-1,0,2") -> (Output_Get:Vector)
   > Set scarf_back_rot(scarf_back_rot:Rotator="-17,-15,-5") -> (Output_Get:Rotator)
   Set scarf_back_trans(execute=@n2.else, scarf_back_trans="-1.5,0,2")
   > Set scarf_back_rot(scarf_back_rot="-32,-20,-15")
   Event BlueprintInitializeAnimation()
   > Set crouchRootSpace(crouchRootSpace:double=1) -> (Output_Get:double)
   > Set crouchSpineSpace(crouchSpineSpace:double=0) -> (Output_Get:double)
   > Set Skel_Mesh(Skel_Mesh=GetOwnerSkeletalMeshComponent())
   Event BlueprintUpdateAnimation()
   > Get!(execute:Exec) -> (then:Exec, else:Exec, Skel_Mesh:Skeletal Mesh Component Object Reference)
   > n3 = Sequence!(execute) -> (then_0, then_1, then_2:Exec, then_3:Exec, then_4:Exec, then_5:Exec)
   n4 = Break Fort Anim Input Skydiving External Force(FortAnimInput_SkydivingExternalForce=Get SkydivingExternalForceInput()) -> (
    bUseSkydivingVectorForce=~
    SkydivingVectorForce=~
    HeadToPelvisDirection=~
    FloatingMultiplier=~
    FloatingAdditive=~
    DivingMultiplier=~
    DivingAdditive=~
    ParachutingMultiplier=~
    ParachutingAdditive=~
   )
   > n5 = Set windVector(windVector:Vector=SelectVector(A=vector * float(A=@n4.SkydivingVectorForce, B=0.15), B="0, 0,-980", bPickA=Get bIsSkyDiving())) -> (Output_Get:Vector)
   n6 = BreakTransform(InTransform=GetDeltaTransformFromRefPose(self=GetOwnerSkeletalMeshComponent(), BoneName=pelvis, BaseName=root))
   n7 = "float > float"(A="Absolute (Float)"(A=BreakRotator(InRot=@n6.Rotation).Yaw), B=90) {is=title K2Node_CallFunction}
   n8 = SelectFloat(A=980, B=-980, bPickA=@n7)
   > Set finalGravity(finalGravity:Vector=SelectVector(A=MakeVector(Y=0, Z=@n8), B=@n5, bPickA=Get bIsPlayingEmote())) -> (Output_Get:Vector)
   > Set isFlipping(isFlipping:bool=@n7) -> (Output_Get:bool)
   n9 = Set skydiveSpineSpace(execute=[@n30.then, @n3.then_1], skydiveSpineSpace:double=SelectFloat(A=1, B=0, bPickA=Get bIsSkyDiving())) -> (Output_Get:double)
   > Set skydiveRootSpace(skydiveRootSpace:double=float - float(B=@n9) {is=title K2Node_CallFunction}) -> (Output_Get:double)
   n10 = Get bIsSkyDiving()
   Set finalRootSpace(
    execute=@n3.then_2
    finalRootSpace:double=SelectFloat(A=Get skydiveRootSpace() -> (skydiveRootSpace:double), B=Get crouchRootSpace() -> (crouchRootSpace:double), bPickA=@n10)
   ) -> (Output_Get:double)
   > Set finalSpineSpace(finalSpineSpace:double=SelectFloat(A=Get skydiveSpineSpace() -> (skydiveSpineSpace:double), B=Get crouchSpineSpace() -> (crouchSpineSpace:double), bPickA=@n10)) -> (Output_Get:double)
   Set thigh_rot_r(
    execute=@n3.then_3
    thigh_rot_r=GetAlphaFromReferenceTransformDelta(
     SkeletalMeshComponent=Get Skel_Mesh()
     BoneName=thigh_r
     BaseName=pelvis
     DeltaComponentType=QuaternionTwist
     InputRange=(X=0,Y=100)
     TwistAxis="0,0,1"
    )
   )
   n11 = Set player_speed(execute=@n3.then_5, player_speed:double=Get PawnSpeed2D()) -> (Output_Get:double)
   > n12 = Branch(Condition="float > float"(A=@n11, B=0))
   > Set pouch_rigids_alpha(pouch_rigids_alpha:double=MapRangeClamped(Value=@n11, InRangeA=50, InRangeB=400)) -> (Output_Get:double)
   n13 = Branch(execute=@n12.else, Condition=OR Boolean(A=Get jump_bool() -> (jump_bool:bool), B=Get fall_bool() -> (fall_bool:bool)))
   > Set pouch_rigids_alpha(pouch_rigids_alpha=0.6)
   Set pouch_rigids_alpha(execute=@n13.else, pouch_rigids_alpha=0.25)
   n14 = Get Skel_Mesh()
   n15 = GetAlphaFromReferenceTransformDelta(SkeletalMeshComponent=@n14, BoneName=head, BaseName=spine_05, DeltaComponentType=QuaternionTwist, InputRange=(X=10,Y=95), TwistAxis="0,0,1")
   Set collar_back_alpha(execute=@n3.then_4, collar_back_alpha:double=@n15) -> (Output_Get:double)
   > Set collar_back_down_alpha(collar_back_down_alpha:double=@n15) -> (Output_Get:double)
   n16 = GetAlphaFromReferenceTransformDelta(SkeletalMeshComponent=@n14, BoneName=head, BaseName=spine_05, DeltaComponentType=QuaternionTwist, InputRange=(X=15,Y=90), TwistAxis="1,0,0")
   > Set collar_back_left_alpha(collar_back_left_alpha:double=@n16) -> (Output_Get:double)
   n17 = float * float(A=@n16, B=0.3)
   n18 = Get collar_back_down_alpha()
   > Set collar_back_down_alpha(collar_back_down_alpha="Clamp (Float)"(Value=float + float(A=@n17, B=@n18) {is=title K2Node_CommutativeAssociativeBinaryOperator}, Min=0, Max=1))
   n19 = GetAlphaFromReferenceTransformDelta(
    SkeletalMeshComponent=@n14
    BoneName=head
    BaseName=spine_05
    DeltaComponentType=QuaternionTwist
    InputRange=(X=-15,Y=-90)
    TwistAxis="1,0,0"
   )
   > Set collar_back_right_alpha(collar_back_right_alpha:double=@n19) -> (Output_Get:double)
   n20 = float * float(A=@n19, B=0.3)
   n21 = Get collar_back_down_alpha()
   > Set collar_back_down_alpha(collar_back_down_alpha="Clamp (Float)"(Value=float + float(A=@n20, B=@n21), Min=0, Max=1))
   n22 = Get Skel_Mesh()
   > Set collar_front_alpha(collar_front_alpha:double=GetAlphaFromReferenceTransformDelta(SkeletalMeshComponent=@n22, BoneName=head, BaseName=spine_05, InputRange=(X=0,Y=25), TwistAxis="1,0,0")) -> (Output_Get:double)
   n23 = GetAlphaFromReferenceTransformDelta(SkeletalMeshComponent=@n22, BoneName=head, BaseName=spine_05, DeltaComponentType=QuaternionTwist, InputRange=(X=0,Y=45), TwistAxis="0,1,0")
   > Set neck_right_alpha(neck_right_alpha:double=@n23) -> (Output_Get:double)
   n24 = Get collar_back_down_alpha()
   n25 = float + float(A=float * float(A=@n24, B=@n23), B=@n24)
   > Set collar_back_down_alpha(collar_back_down_alpha="Clamp (Float)"(Value="Clamp (Float)"(Value=@n25, Min=0, Max=1), Min=0, Max=1))
   n26 = float * float(A=@n23, B=0.7)
   n27 = Get collar_back_left_alpha()
   > Set collar_back_left_alpha(collar_back_left_alpha="Clamp (Float)"(Value=float + float(A=@n26, B=@n27), Min=0, Max=1))
   > Branch(Condition=Get crouch_bool() -> (crouch_bool:bool))
   n28 = MapRangeClamped(Value=Get PawnSpeed2D(), InRangeA=50, InRangeB=140)
   n29 = float * float(A=Get collar_back_down_alpha(), B=1.35)
   > Set collar_crouch_down_alpha(collar_crouch_down_alpha:double="Clamp (Float)"(Value=float * float(A=@n29, B=@n28), Min=0, Max=1)) -> (Output_Get:double)
   Event OnEndCrouching()
   > Set crouch_bool(crouch_bool:bool=false) -> (Output_Get:bool)
   > Set crouchRootSpace(crouchRootSpace=0)
   > Set crouchSpineSpace(crouchSpineSpace=1)
   > Delay(Duration=0.2)
   > Set crouchRootSpace(crouchRootSpace=1)
   > Set crouchSpineSpace(crouchSpineSpace=0)
   > ResetDynamics(InTeleportType=TeleportPhysics)
   Event OnBeginFalling()
   > Set fall_bool(fall_bool:bool=true) -> (Output_Get:bool)
   Event OnBeginCrouching()
   > Set crouch_bool(crouch_bool=true)
   > Set crouchRootSpace(crouchRootSpace=0)
   > Set crouchSpineSpace(crouchSpineSpace=1)
   > ResetDynamics(InTeleportType=TeleportPhysics)
   > Delay(Duration=0.2)
   > Set crouchRootSpace(crouchRootSpace=1)
   > Set crouchSpineSpace(crouchSpineSpace=0)
   Event OnEndFalling()
   > Set fall_bool(fall_bool=false)
   Event OnBeginSkydiving()
   > ResetDynamics(InTeleportType=TeleportPhysics)
   > n30 = ResetDynamics(InTeleportType=ResetPhysics)
   Event OnEndJumping()
   > Set jump_bool(jump_bool=false)
   Event OnEndSkydiving()
   > ResetDynamics(InTeleportType=TeleportPhysics)
   > ResetDynamics(InTeleportType=ResetPhysics)
   n31 = Set maskless_active(maskless_active:bool=@n1) -> (Output_Get:bool)
   Reroute Node(InputPin:Wildcard) -> (OutputPin:Wildcard)
  ]}
]
variables [
 collar_back_alpha {type=number}
 collar_back_down_alpha {type=number}
 collar_back_left_alpha {type=number}
 collar_back_right_alpha {type=number}
 collar_crouch_down_alpha {type=number}
 collar_front_alpha {type=number}
 crouch_bool {type=boolean}
 crouchRootSpace {type=number}
 crouchSpineSpace {type=number}
 fall_bool {type=boolean}
 finalGravity {type=Vector}
 finalRootSpace {type=number}
 finalSpineSpace {type=number}
 isFlipping {type=boolean}
 jump_bool {type=boolean}
 maskless_active {type=boolean}
 neck_right_alpha {type=number}
 player_speed {type=number}
 pouch_rigids_alpha {type=number}
 scarf_back_rot {type=Rotator, default=<rot -32 -20 -15>}
 scarf_back_trans {type=Vector, default=<-1.5 0 15>}
 Skel_Mesh {type=SkeletalMeshComponent}
 SkydiveRootSpace {type=number}
 SkydiveSpineSpace {type=number}
 thigh_rot_r {type=number}
 windVector {type=Vector}
]
